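# Physics asset for the offhand axe mesh of the BlessFlan pickaxe, a cosmetic melee weapon.
# It defines a single capsule body on the Melee_TwoHanded_Handle bone, offset and rotated relative to that bone.
# It has no constraints, so the collision is one simple shape with no jointed parts.
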PhysicsAsset /BRCosmetics/Weapons/FORT_Melee/Pickaxe_BlessFlan/Meshes/Offhand/BlessFlan_Offhand_Axe_PhysicsAsset
bodies {Melee_TwoHanded_Handle [{shapeType=Capsule, length=44.9916, radius=5.3642, center=<1.0025 1.8458 29.2496>, rotation=<rot 4.5832 -90 0.0217>}]}
constraints []
